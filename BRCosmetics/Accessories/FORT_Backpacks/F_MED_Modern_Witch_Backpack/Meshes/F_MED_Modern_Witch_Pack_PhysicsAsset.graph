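# Physics asset for the Modern Witch backpack cosmetic, giving box bodies and joints to the pack, its handle and two straps.
# A large pack box anchors small handle and strap master bodies through joints with free swing and twist.
# Three-segment chains hang from the masters: one for the handle, two for straps sharing a master.
# Chain joints use angular limits, loosening toward the handle tip and keeping strap twist tight.
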PhysicsAsset /BRCosmetics/Accessories/FORT_Backpacks/F_MED_Modern_Witch_Backpack/Meshes/F_MED_Modern_Witch_Pack_PhysicsAsset
bodies {
 dyn_handle_01 [{shapeType=Box, extentX=2, extentY=2, extentZ=2, center=<1 0 0>, rotation=<rot 90 0 0>}]
 dyn_handle_02 [{shapeType=Box, extentX=2, extentY=2, extentZ=2, center=<1 0 0>, rotation=<rot 90 0 0>}]
 dyn_handle_03 [{shapeType=Box, extentX=2, extentY=2, extentZ=2, center=<1 0 0>, rotation=<rot 90 0 0>}]
 dyn_handle_mstr [{shapeType=Box, extentX=1.01, extentY=1.01, extentZ=1.01}]
 dyn_strap_a_01 [{shapeType=Box, extentX=1.1, extentY=1.5, extentZ=2, center=<1 0 0>, rotation=<rot 90 0 0>}]
 dyn_strap_a_02 [{shapeType=Box, extentX=1.1, extentY=1.5, extentZ=2, center=<1 0 0>, rotation=<rot 90 0 0>}]
 dyn_strap_a_03 [{shapeType=Box, extentX=1.1, extentY=2.2, extentZ=2, center=<0.6 0 0>, rotation=<rot 90 0 0>}]
 dyn_strap_b_01 [{shapeType=Box, extentX=1.1, extentY=1.5, extentZ=2, center=<1 0 0>, rotation=<rot 90 0 0>}]
 dyn_strap_b_02 [{shapeType=Box, extentX=1.1, extentY=1.5, extentZ=2, center=<1 0 0>, rotation=<rot 90 0 0>}]
 dyn_strap_b_03 [{shapeType=Box, extentX=1.1, extentY=2.2, extentZ=2, center=<0.6 0 0>, rotation=<rot 90 0 0>}]
 dyn_strap_mstr [{shapeType=Box, extentX=1.01, extentY=1.01, extentZ=1.01}]
 pack [{shapeType=Box, extentX=20.1389, extentY=9.6576, extentZ=21.1199, center=<0 -2.5168 -1.7873>, rotation=<rot 0 0 -1.8982>}]
}
constraints [
 {
  bone1Name=dyn_strap_mstr
  bone2Name=pack
  swing1LimitDegrees=45
  swing1Motion=Free
  swing2LimitDegrees=45
  swing2Motion=Free
  twistLimitDegrees=45
  twistMotion=Free
 }
 {
  bone1Name=dyn_handle_mstr
  bone2Name=pack
  swing1LimitDegrees=45
  swing1Motion=Free
  swing2LimitDegrees=45
  swing2Motion=Free
  twistLimitDegrees=45
  twistMotion=Free
 }
 {bone1Name=dyn_strap_a_01, bone2Name=dyn_strap_mstr, swing1LimitDegrees=20, swing2LimitDegrees=45, twistLimitDegrees=5}
 {bone1Name=dyn_strap_a_02, bone2Name=dyn_strap_a_01, swing1LimitDegrees=20, swing2LimitDegrees=30, twistLimitDegrees=5}
 {bone1Name=dyn_strap_a_03, bone2Name=dyn_strap_a_02, swing1LimitDegrees=20, swing2LimitDegrees=30, twistLimitDegrees=5}
 {bone1Name=dyn_strap_b_01, bone2Name=dyn_strap_mstr, swing1LimitDegrees=20, swing2LimitDegrees=45, twistLimitDegrees=5}
 {bone1Name=dyn_strap_b_02, bone2Name=dyn_strap_b_01, swing1LimitDegrees=20, swing2LimitDegrees=30, twistLimitDegrees=5}
 {bone1Name=dyn_strap_b_03, bone2Name=dyn_strap_b_02, swing1LimitDegrees=20, swing2LimitDegrees=30, twistLimitDegrees=5}
 {bone1Name=dyn_handle_01, bone2Name=dyn_handle_mstr, swing1LimitDegrees=15, swing2LimitDegrees=15, twistLimitDegrees=5}
 {bone1Name=dyn_handle_02, bone2Name=dyn_handle_01, swing1LimitDegrees=20, swing2LimitDegrees=20, twistLimitDegrees=7.5}
 {bone1Name=dyn_handle_03, bone2Name=dyn_handle_02, swing1LimitDegrees=30, swing2LimitDegrees=30, twistLimitDegrees=15}
]
