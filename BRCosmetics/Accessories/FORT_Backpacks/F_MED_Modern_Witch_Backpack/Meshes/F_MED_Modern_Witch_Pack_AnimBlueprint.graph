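# Animation Blueprint for the Modern Witch backpack cosmetic that adds secondary physics motion to its straps and handles.
# The cached input pose feeds a gravity-corrector RigidBody and an AnimDynamics strap-and-handle chain, joined by a bool blend into the output.
# The event graph updates GravityOverride through a deprecated function and sets a clamp vector on targeting events.
# An alternate bone-relative strap chain, a second RigidBody and VInterpTo nodes are left disconnected.
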AnimBlueprint /BRCosmetics/Accessories/FORT_Backpacks/F_MED_Modern_Witch_Backpack/Meshes/F_MED_Modern_Witch_Pack_AnimBlueprint
target_skeleton=&/F_MED_Modern_Witch_Pack_Skeleton
graphs [
 AnimGraph {nodes [
   n1 = Anim Dynamics - Bone: dyn_strap_a_01!(ComponentPose, Alpha=1, GravityOverride:Vector="0,0,-980") -> (Pose) {
    angularDampingOverride=0.999
    angularSpringConstant=1000
    bAngularSpring=true
    bOverrideAngularDamping=true
    bUseGravityOverride=true
    gravityScale=0.35
    simulationSpace=BoneRelative
    alphaBoolBlend {blendOption=Linear}
    boundBone {boneName=dyn_strap_a_01}
    componentAppliedLinearAccClamp=<5000 5000 5000>
    componentLinearAccScale=<1 1 1>
    componentLinearVelScale=<1 1 1>
    gravityOverride {z=-980}
    physicsBodyDefinitions [{
      boundBone {boneName=dyn_strap_a_01}
      boxExtents=<8 12 15>
      constraintSetup {angularLimitsMax=<0 0 45>, angularLimitsMin=<0 -20 -55>, angularTarget=<1 0 0>}
      localJointOffset=<0.5 0 0>
     }]
    relativeSpaceBone {boneName=pack}
    simSpaceSettings {maxAngularAcceleration=2, maxAngularVelocity=2, simSpaceAngularAlpha=1}
   }
   n2 = Anim Dynamics - Bone: dyn_strap_a_02!(ComponentPose=@n1, Alpha=1, GravityOverride:Vector="0,0,-980") -> (Pose) {
    angularDampingOverride=0.999
    angularSpringConstant=500
    bAngularSpring=true
    bOverrideAngularDamping=true
    bUseGravityOverride=true
    gravityScale=0.35
    simulationSpace=BoneRelative
    alphaBoolBlend {blendOption=Linear}
    boundBone {boneName=dyn_strap_a_02}
    componentAppliedLinearAccClamp=<5000 5000 5000>
    componentLinearAccScale=<1 1 1>
    componentLinearVelScale=<1 1 1>
    gravityOverride {z=-980}
    physicsBodyDefinitions [{
      boundBone {boneName=dyn_strap_a_02}
      boxExtents=<8 12 15>
      constraintSetup {angularLimitsMax=<0 5 45>, angularLimitsMin=<0 -35 -55>, angularTarget=<1 0 0>}
      localJointOffset=<0.5 0 0>
     }]
    relativeSpaceBone {boneName=pack}
    simSpaceSettings {maxAngularAcceleration=2, maxAngularVelocity=2, simSpaceAngularAlpha=1}
   }
   n3 = Anim Dynamics - Bone: dyn_strap_a_03!(ComponentPose=@n2, Alpha=1, GravityOverride:Vector="0,0,-980") -> (Pose) {
    angularDampingOverride=0.999
    bAngularSpring=true
    bOverrideAngularDamping=true
    bUseGravityOverride=true
    gravityScale=0.35
    simulationSpace=BoneRelative
    alphaBoolBlend {blendOption=Linear}
    boundBone {boneName=dyn_strap_a_03}
    componentAppliedLinearAccClamp=<5000 5000 5000>
    componentLinearAccScale=<1 1 1>
    componentLinearVelScale=<1 1 1>
    gravityOverride {z=-980}
    physicsBodyDefinitions [{
      boundBone {boneName=dyn_strap_a_03}
      boxExtents=<8 12 12>
      constraintSetup {angularLimitsMax=<0 5 45>, angularLimitsMin=<0 -35 -55>, angularTarget=<1 0 0>}
      localJointOffset=<1.5 0 0>
     }]
    relativeSpaceBone {boneName=pack}
    simSpaceSettings {maxAngularAcceleration=2, maxAngularVelocity=2, simSpaceAngularAlpha=1}
   }
   n4 = Anim Dynamics - Bone: dyn_strap_b_01!(ComponentPose=@n3, Alpha=1, GravityOverride:Vector="0,0,-980") -> (Pose) {
    angularDampingOverride=0.999
    angularSpringConstant=1000
    bAngularSpring=true
    bOverrideAngularDamping=true
    bUseGravityOverride=true
    gravityScale=0.35
    simulationSpace=BoneRelative
    alphaBoolBlend {blendOption=Linear}
    boundBone {boneName=dyn_strap_b_01}
    componentAppliedLinearAccClamp=<5000 5000 5000>
    componentLinearAccScale=<1 1 1>
    componentLinearVelScale=<1 1 1>
    gravityOverride {z=-980}
    physicsBodyDefinitions [{
      boundBone {boneName=dyn_strap_b_01}
      boxExtents=<8 12 15>
      constraintSetup {angularLimitsMax=<0 0 45>, angularLimitsMin=<0 -20 -65>, angularTarget=<1 0 0>}
      localJointOffset=<0.4 0 0>
     }]
    relativeSpaceBone {boneName=pack}
    simSpaceSettings {maxAngularAcceleration=2, maxAngularVelocity=2, simSpaceAngularAlpha=1}
   }
   n5 = Anim Dynamics - Bone: dyn_strap_b_02!(ComponentPose=@n4, Alpha=1, GravityOverride:Vector="0,0,-980") -> (Pose) {
    angularDampingOverride=0.999
    angularSpringConstant=500
    bAngularSpring=true
    bOverrideAngularDamping=true
    bUseGravityOverride=true
    gravityScale=0.35
    simulationSpace=BoneRelative
    alphaBoolBlend {blendOption=Linear}
    boundBone {boneName=dyn_strap_b_02}
    componentAppliedLinearAccClamp=<5000 5000 5000>
    componentLinearAccScale=<1 1 1>
    componentLinearVelScale=<1 1 1>
    gravityOverride {z=-980}
    physicsBodyDefinitions [{
      boundBone {boneName=dyn_strap_b_02}
      boxExtents=<8 12 15>
      constraintSetup {angularLimitsMax=<0 5 45>, angularLimitsMin=<0 -35 -65>, angularTarget=<1 0 0>}
      localJointOffset=<0.6 0 0>
     }]
    relativeSpaceBone {boneName=pack}
    simSpaceSettings {maxAngularAcceleration=2, maxAngularVelocity=2, simSpaceAngularAlpha=1}
   }
   Anim Dynamics - Bone: dyn_strap_b_03!(ComponentPose=@n5, Alpha=1, GravityOverride:Vector="0,0,-980") -> (Pose) {
    angularDampingOverride=0.999
    bAngularSpring=true
    bOverrideAngularDamping=true
    bUseGravityOverride=true
    gravityScale=0.35
    simulationSpace=BoneRelative
    alphaBoolBlend {blendOption=Linear}
    boundBone {boneName=dyn_strap_b_03}
    componentAppliedLinearAccClamp=<5000 5000 5000>
    componentLinearAccScale=<1 1 1>
    componentLinearVelScale=<1 1 1>
    gravityOverride {z=-980}
    physicsBodyDefinitions [{
      boundBone {boneName=dyn_strap_b_03}
      boxExtents=<8 12 12>
      constraintSetup {angularLimitsMax=<0 5 45>, angularLimitsMin=<0 -35 -65>, angularTarget=<1 0 0>}
      localJointOffset=<1.2 0 0>
     }]
    relativeSpaceBone {boneName=pack}
    simSpaceSettings {maxAngularAcceleration=2, maxAngularVelocity=2, simSpaceAngularAlpha=1}
   }
   Anim Dynamics - Bone: dyn_strap_b_03(Alpha=0.4, GravityOverride="0,0,-980") {
    angularDampingOverride=1
    angularSpringConstant=20000
    bOverrideAngularDamping=true
    bUseGravityOverride=true
    gravityScale=0.35
    simulationSpace=RootRelative
    alphaBoolBlend {blendOption=Linear}
    boundBone {boneName=dyn_strap_b_03}
    componentAppliedLinearAccClamp=<3000 3000 3000>
    componentLinearAccScale=<0.6 0.6 0.6>
    componentLinearVelScale=<0.6 0.6 0.6>
    gravityOverride {z=-980}
    physicsBodyDefinitions [{
      boundBone {boneName=dyn_strap_b_03}
      boxExtents=<10 16 8>
      constraintSetup {angularLimitsMax=<0 5 45>, angularLimitsMin=<0 -35 -65>, angularTarget=<1 0 0>}
      localJointOffset=<3 0 0>
     }]
    simSpaceSettings {maxAngularAcceleration=2, maxAngularVelocity=1, simSpaceAngularAlpha=1}
   }
   RigidBody!(ComponentPose, bAlphaBoolEnabled:bool, ExternalForce) -> (Pose) {
    alphaInputType=Bool
    bClampLinearTranslationLimitToRefPose=true
    bOverrideWorldGravity=true
    overridePhysicsAsset=&/F_MED_Modern_Witch_Pack_PhysicsAsset
    simulationSpace=BaseBoneSpace
    alphaBoolBlend {blendOption=Linear}
    alphaScaleBiasClamp {bClampResult=true}
    baseBoneRef {boneName=pack}
    componentAppliedLinearAccClamp=<3000 3000 3000>
    componentLinearAccScale=<0.3 0.3 0.3>
    componentLinearVelScale=<0.3 0.3 0.3>
    overrideWorldGravity {z=-980}
   }
   Output Pose(Result=Use cached pose 'Dynamic Cache'()) {layerGroup=None}
   Save cached pose 'Input_Cache'(Pose=Input Pose()) {cacheName=Input_Cache}
   n6 = RigidBody!(ComponentPose=Local To Component(LocalPose=Use cached pose 'Input_Cache'()), Alpha:float, OverrideWorldGravity:Vector="0,0,0") -> (Pose) {
    bOverrideWorldGravity=true
    overridePhysicsAsset=&/F_MED_Modern_Witch_Pack_GravityCorrector_Physics
    simulationSpace=BaseBoneSpace
    alphaBoolBlend {blendOption=Linear}
    baseBoneRef {boneName=pack}
    componentAppliedLinearAccClamp=<5000 5000 5000>
    componentLinearAccScale=<1 1 0.3>
    componentLinearVelScale=<1 1 0.3>
   }
   n7 = Anim Dynamics - Bone: dyn_strap_a_01(ComponentPose=Local To Component(LocalPose=Use cached pose 'Input_Cache'()), Alpha=0.4, GravityOverride="0,0,-980") {
    angularDampingOverride=1
    angularSpringConstant=20000
    bOverrideAngularDamping=true
    bUseGravityOverride=true
    gravityScale=0.35
    simulationSpace=RootRelative
    alphaBoolBlend {blendOption=Linear}
    boundBone {boneName=dyn_strap_a_01}
    componentAppliedLinearAccClamp=<3000 3000 3000>
    componentLinearAccScale=<0.6 0.6 0.6>
    componentLinearVelScale=<0.6 0.6 0.6>
    gravityOverride {z=-980}
    physicsBodyDefinitions [{
      boundBone {boneName=dyn_strap_a_01}
      boxExtents=<10 16 8>
      constraintSetup {angularLimitsMax=<0 0 45>, angularLimitsMin=<0 -20 -55>, angularTarget=<1 0 0>}
      localJointOffset=<3 0 0>
     }]
    simSpaceSettings {maxAngularAcceleration=2, maxAngularVelocity=1, simSpaceAngularAlpha=1}
   }
   n8 = Anim Dynamics - Bone: dyn_strap_a_02(ComponentPose=@n7, Alpha=0.4, GravityOverride="0,0,-980") {
    angularDampingOverride=1
    angularSpringConstant=20000
    bOverrideAngularDamping=true
    bUseGravityOverride=true
    gravityScale=0.35
    simulationSpace=RootRelative
    alphaBoolBlend {blendOption=Linear}
    boundBone {boneName=dyn_strap_a_02}
    componentAppliedLinearAccClamp=<3000 3000 3000>
    componentLinearAccScale=<0.6 0.6 0.6>
    componentLinearVelScale=<0.6 0.6 0.6>
    gravityOverride {z=-980}
    physicsBodyDefinitions [{
      boundBone {boneName=dyn_strap_a_02}
      boxExtents=<10 16 8>
      constraintSetup {angularLimitsMax=<0 5 45>, angularLimitsMin=<0 -35 -55>, angularTarget=<1 0 0>}
      localJointOffset=<3 0 0>
     }]
    simSpaceSettings {maxAngularAcceleration=2, maxAngularVelocity=1, simSpaceAngularAlpha=1}
   }
   n9 = Anim Dynamics - Bone: dyn_strap_a_03(ComponentPose=@n8, Alpha=0.4, GravityOverride="0,0,-980") {
    angularDampingOverride=1
    angularSpringConstant=20000
    bOverrideAngularDamping=true
    bUseGravityOverride=true
    gravityScale=0.35
    simulationSpace=RootRelative
    alphaBoolBlend {blendOption=Linear}
    boundBone {boneName=dyn_strap_a_03}
    componentAppliedLinearAccClamp=<3000 3000 3000>
    componentLinearAccScale=<0.6 0.6 0.6>
    componentLinearVelScale=<0.6 0.6 0.6>
    gravityOverride {z=-980}
    physicsBodyDefinitions [{
      boundBone {boneName=dyn_strap_a_03}
      boxExtents=<10 16 8>
      constraintSetup {angularLimitsMax=<0 5 45>, angularLimitsMin=<0 -35 -55>, angularTarget=<1 0 0>}
      localJointOffset=<3 0 0>
     }]
    simSpaceSettings {maxAngularAcceleration=2, maxAngularVelocity=1, simSpaceAngularAlpha=1}
   }
   n10 = Anim Dynamics - Bone: dyn_strap_b_01(ComponentPose=@n9, Alpha=0.4, GravityOverride="0,0,-980") {
    angularDampingOverride=1
    angularSpringConstant=20000
    bOverrideAngularDamping=true
    bUseGravityOverride=true
    gravityScale=0.35
    simulationSpace=RootRelative
    alphaBoolBlend {blendOption=Linear}
    boundBone {boneName=dyn_strap_b_01}
    componentAppliedLinearAccClamp=<3000 3000 3000>
    componentLinearAccScale=<0.6 0.6 0.6>
    componentLinearVelScale=<0.6 0.6 0.6>
    gravityOverride {z=-980}
    physicsBodyDefinitions [{
      boundBone {boneName=dyn_strap_b_01}
      boxExtents=<10 16 8>
      constraintSetup {angularLimitsMax=<0 0 45>, angularLimitsMin=<0 -20 -65>, angularTarget=<1 0 0>}
      localJointOffset=<3 0 0>
     }]
    simSpaceSettings {maxAngularAcceleration=2, maxAngularVelocity=1, simSpaceAngularAlpha=1}
   }
   n11 = Anim Dynamics - Bone: dyn_strap_b_02(ComponentPose=@n10, Alpha=0.4, GravityOverride="0,0,-980") {
    angularDampingOverride=1
    angularSpringConstant=20000
    bOverrideAngularDamping=true
    bUseGravityOverride=true
    gravityScale=0.35
    simulationSpace=RootRelative
    alphaBoolBlend {blendOption=Linear}
    boundBone {boneName=dyn_strap_b_02}
    componentAppliedLinearAccClamp=<3000 3000 3000>
    componentLinearAccScale=<0.6 0.6 0.6>
    componentLinearVelScale=<0.6 0.6 0.6>
    gravityOverride {z=-980}
    physicsBodyDefinitions [{
      boundBone {boneName=dyn_strap_b_02}
      boxExtents=<10 16 8>
      constraintSetup {angularLimitsMax=<0 5 45>, angularLimitsMin=<0 -35 -65>, angularTarget=<1 0 0>}
      localJointOffset=<3 0 0>
     }]
    simSpaceSettings {maxAngularAcceleration=2, maxAngularVelocity=1, simSpaceAngularAlpha=1}
   }
   n12 = Anim Dynamics - Bone: dyn_handle_01!(ComponentPose=@n11, Alpha=1, GravityOverride:Vector="0,0,0") -> (Pose) {
    angularDampingOverride=1
    angularSpringConstant=15000
    bAngularSpring=true
    bOverrideAngularDamping=true
    bUseGravityOverride=true
    gravityScale=0.35
    simulationSpace=RootRelative
    alphaBoolBlend {blendOption=Linear}
    boundBone {boneName=dyn_handle_01}
    componentAppliedLinearAccClamp=<10000 10000 10000>
    componentLinearAccScale=<1 1 1>
    componentLinearVelScale=<1 1 1>
    gravityOverride {z=-980}
    physicsBodyDefinitions [{
      boundBone {boneName=dyn_handle_01}
      boxExtents=<25 35 35>
      constraintSetup {angularLimitsMax=<0 0 50>, angularLimitsMin=<0 0 -50>, angularTarget=<1 0 0>}
      localJointOffset=<1 0 0>
     }]
    relativeSpaceBone {boneName=pack}
    simSpaceSettings {maxAngularAcceleration=36, maxAngularVelocity=6, simSpaceAngularAlpha=1}
   }
   n13 = Anim Dynamics - Bone: dyn_handle_02!(ComponentPose=@n12, Alpha=1, GravityOverride:Vector="0,0,0") -> (Pose) {
    angularDampingOverride=1
    angularSpringConstant=15000
    bAngularSpring=true
    bOverrideAngularDamping=true
    bUseGravityOverride=true
    gravityScale=0.35
    simulationSpace=RootRelative
    alphaBoolBlend {blendOption=Linear}
    boundBone {boneName=dyn_handle_02}
    componentAppliedLinearAccClamp=<10000 10000 10000>
    componentLinearAccScale=<1 1 1>
    componentLinearVelScale=<1 1 1>
    gravityOverride {z=-980}
    physicsBodyDefinitions [{
      boundBone {boneName=dyn_handle_02}
      boxExtents=<25 35 35>
      constraintSetup {angularLimitsMax=<0 0 50>, angularLimitsMin=<0 0 -50>, angularTarget=<1 0 0>}
      localJointOffset=<1 0 0>
     }]
    relativeSpaceBone {boneName=pack}
    simSpaceSettings {maxAngularAcceleration=36, maxAngularVelocity=6, simSpaceAngularAlpha=1}
   }
   Save cached pose 'Dynamic Cache'(Pose=Blend Poses by bool(BlendPose_0=Component To Local(ComponentPose=@n6), BlendPose_1=Component To Local(ComponentPose=@n13)) {blendType=Linear}) {cacheName=Dynamic Cache}
  ]}
 EventGraph {nodes [
   Event OnBeginTargeting() {is=title K2Node_Event}
   > Set BeginTarget_Clamp(BeginTarget_Clamp:Vector="0.125,0.125,0.5") -> (Output_Get:Vector)
   Event OnEndTargeting() {is=title K2Node_Event}
   > Set BeginTarget_Clamp(BeginTarget_Clamp="0.6,0.6,0.9")
   n1 = Event BlueprintUpdateAnimation()
   > n2 = Gravity Override State Function DEPRECATED(CharacterPartAnimInstance=Self-Reference(), GameTimeInSeconds=GetGameTimeInSeconds(), GravityOverrideParams=Get Gravity Override Params())
   > Set GravityOverride(GravityOverride=@n2)
   VInterpTo!(Current:Vector="0, 0, 0", Target:Vector="0.6,0.6,0.6", DeltaTime:float=@n1.DeltaTimeX, InterpSpeed:float=50) -> (ReturnValue:Vector)
   VInterpTo(Current="0, 0, 0", Target="0.2,0.2,0.2", DeltaTime=@n1.DeltaTimeX, InterpSpeed=20)
   TryGetPawnOwner()
   Get bIsTargeting()
  ]}
]
variables [
 BeginTarget_Clamp {type=Vector}
 Gravity Override Params {type=GravityOverrideParamsStruct}
 GravityOverride {type=Vector}
]
